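# Physics asset for the SpireTend gadget prop, defining collision bodies and joints for a dangling chain.
# Capsules cover the base and first five dynamic links, a thin box caps the last link, and key_jnt has a large unconstrained capsule.
# Six constraints join the links in series from base_chain, with wide swing at the root, tighter swing down the chain and very little twist throughout.
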PhysicsAsset /BRCosmetics/Animation/Props/Gadgets_SpireTrend/Meshes/SpireTend_Gadget_Physics
bodies {
 base_chain [{shapeType=Capsule, length=0.505, radius=4.005, shapeName=base_chain_capsule, center=<-6.0483 0 0>, rotation=<rot 0 -90 0>}]
 dyn_chain_01 [{
   shapeType=Capsule
   length=2.9965
   radius=1.9713
   shapeName=dyn_chain_01_capsule
   center=<2.011 -0.0005 0.0672>
   rotation=<rot 0 90.0143 -88.4496>
  }]
 dyn_chain_02 [{
   shapeType=Capsule
   length=3.1463
   radius=1.2348
   shapeName=dyn_chain_02_capsule
   center=<2.5091 -0.0015 -0.0327>
   rotation=<rot 0 89.8923 -89.1512>
  }]
 dyn_chain_03 [{
   shapeType=Capsule
   length=3.1463
   radius=1.2348
   shapeName=dyn_chain_03_capsule
   center=<2.5756 0.0013 0.0325>
   rotation=<rot 0 -89.8922 -89.1512>
  }]
 dyn_chain_04 [{
   shapeType=Capsule
   length=3.1463
   radius=1.2348
   shapeName=dyn_chain_04_capsule
   center=<2.5513 -0.0015 -0.0327>
   rotation=<rot 0 89.8923 -89.1512>
  }]
 dyn_chain_05 [{shapeType=Capsule, length=2.1626, radius=1.2495, shapeName=dyn_chain_05_capsule, center=<1.9375 -0.004 0.0001>, rotation=<rot 0 88.0657 -90>}]
 dyn_chain_06 [{
   shapeType=Box
   extentX=1.01
   extentY=11.0558
   extentZ=11.7485
   shapeName=dyn_chain_06_box
   center=<4.9351 0 1.7729>
   rotation=<rot 0 90.0007 -89.8186>
  }]
 key_jnt [{
   shapeType=Capsule
   length=10.7581
   radius=15.0353
   shapeName=key_jnt_capsule
   center=<-0.112 -0.3404 2.8542>
   rotation=<rot 0.7843 -89.9935 0.468>
  }]
}
constraints [
 {bone1Name=dyn_chain_01, bone2Name=base_chain, swing1LimitDegrees=90, swing2LimitDegrees=90, twistLimitDegrees=2}
 {bone1Name=dyn_chain_02, bone2Name=dyn_chain_01, swing1LimitDegrees=35, swing2LimitDegrees=35, twistLimitDegrees=2}
 {bone1Name=dyn_chain_03, bone2Name=dyn_chain_02, swing1LimitDegrees=25, swing2LimitDegrees=25, twistLimitDegrees=2}
 {bone1Name=dyn_chain_04, bone2Name=dyn_chain_03, swing1LimitDegrees=25, swing2LimitDegrees=25, twistLimitDegrees=2}
 {bone1Name=dyn_chain_05, bone2Name=dyn_chain_04, swing1LimitDegrees=25, swing2LimitDegrees=25, twistLimitDegrees=2}
 {bone1Name=dyn_chain_06, bone2Name=dyn_chain_05, swing1LimitDegrees=25, swing2LimitDegrees=25, twistLimitDegrees=2}
]
